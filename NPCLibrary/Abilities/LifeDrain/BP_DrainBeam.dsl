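; Life-drain beam actor that visually links a draining monster to its target player.
; On clients, Tick pins the Niagara beam's ends to a monster socket and the player's spine, and once adds a delayed, local looping "drained" gameplay cue.
; RemoveBeam clears the cue and signals Niagara, then deactivates and destroys the actor after a delay.
; As failsafes against lingering effects, monster death triggers RemoveBeam and client EndPlay removes any active cue.
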(event Collision|EventActorBeginOverlap
  (params (OtherActor :type Actor%20Object%20Reference)))

(event EventTick
  (params (DeltaSeconds :type Float%20%28single-precision%29))
  (bind _ns_lifedrain_beam_loop (Variables|Default|GetNS_LifeDrain_Beam_Loop))
  (bind _monster (Variables|InternalData|GetMonster))
  (Networking|SwitchHasAuthority
    (:Remote
      (Effects|Components|ParticleSystem|SetVectorParameter _ns_lifedrain_beam_loop "SplineUpdatedStartPos" (Transformation|GetSocketLocation (Variables|/Script/Engine.Character|GetMesh (Utilities|Casting|CastToFortPlayerPawn _monster)) (Variables|Default|GetMonsterSocketName)))
      (bind _player (Variables|InternalData|GetPlayer)
        (:then
          (Utilities|FlowControl|Sequence
            (:then_0
              (Networking|SwitchHasAuthority
                (:Remote
                  (Variables|InternalData|SetDrained%20Cue%20Parameters (Ability|GameplayCue|MakeGameplayCueParameters :NormalizedMagnitude 0.0 :RawMagnitude 0.0 :Location "0,0,0" :Normal "0,0,0" :SourceObject _monster :GameplayEffectLevel 1 :AbilityLevel 1 :bReplicateLocationWhenUsingMinimalRepProxy false))
                  (Utilities|FlowControl|DoOnce false
                    (:Completed
                      (Utilities|FlowControl|Delay 0.35)
                      (if (Variables|Default|GetCanAddBeam)
                        (Projectile|AddGameplayCue%28Looping%29Local (Variables|InternalData|GetPlayer) (Variables|Default|GetDrainedGameplayCueTag) (Variables|InternalData|GetDrained%20Cue%20Parameters) false)))))))
            (:then_1
              (if (Math|Vector|NotEqual(Vector) (Transformation|GetSocketLocation (Variables|/Script/Engine.Character|GetMesh _player) "spine_05") "0, 0, 0" 0.0001)
                (Effects|Components|ParticleSystem|SetVectorParameter _ns_lifedrain_beam_loop "SplineUpdatedEndPos" (Transformation|GetSocketLocation (Variables|/Script/Engine.Character|GetMesh _player) "spine_05"))))))
        (:else)))))

(event Custom|OnTarget)

(event Custom|RemoveBeam
  (Utilities|FlowControl|DoOnce false
    (:Completed
      (Variables|Default|SetCanAddBeam false)
      (bind _player (Variables|InternalData|GetPlayer)
        (:then
          (Projectile|RemoveGameplayCue%28Looping%29Local _player (Variables|Default|GetDrainedGameplayCueTag) (Variables|InternalData|GetDrained%20Cue%20Parameters) false))
        (:else
          _))
      (Niagara|SetNiagaraVariable(Float) (Variables|Default|GetNS_LifeDrain_Beam_Loop) "RemoveBeam" 1.0)
      (Utilities|FlowControl|Delay 2.0)
      (Components|Activation|Deactivate (Variables|Default|GetNS_LifeDrain_Beam_Loop))
      (Actor|DestroyActor))))

(event Custom|TrackMonsterDied
  (bind _monster (Variables|InternalData|GetMonster))
  ;[0.4219 1 0.459] There was an issue where sometimes when the Boss dies, the Beam VFX would still remain on players. This is another Failsafe to make sure Bbeams are always removed.
  (Utilities|IsValid _monster
    (:Is%20Valid
      (bind _asfort_pawn (Utilities|Casting|CastToFortPawn _monster))
      (EventDispatchers|BindEventtoOnDied _asfort_pawn (delegate Custom|MonsterDied)))))

(event Custom|MonsterDied
  (params (DamagedActor :type Actor%20Object%20Reference) (Damage :type Float%20%28single-precision%29) (InstigatedBy :type Controller%20Object%20Reference) (DamageCauser :type Actor%20Object%20Reference) (HitLocation :type Vector) (FHitComponent :type Primitive%20Component%20Object%20Reference) (BoneName :type Name) (Momentum :type Vector))
  ;[0.4219 1 0.459] There was an issue where sometimes when the Boss dies, the Beam VFX would still remain on players. This is another Failsafe to make sure Bbeams are always removed.
  (Variables|Default|SetCanAddBeam false)
  (CallFunction|RemoveBeam))

(event EventEndPlay
  (params (EndPlayReason :type EEndPlayReason%20Enum :default "Destroyed"))
  (bind _drainedgameplaycuetag (Variables|Default|GetDrainedGameplayCueTag))
  ;[1 1 1] Add another way to force remove the GC when the Boss dies / Beam is removed.
  (Networking|SwitchHasAuthority
    (:Authority
      (Variables|Default|SetCanAddBeam false))
    (:Remote
      (bind _player (Variables|InternalData|GetPlayer)
        (:then
          (bind _returnvalue (Ability|GameplayCue|IsGameplayCueActive _player _drainedgameplaycuetag))
          (if _returnvalue
            (Projectile|RemoveGameplayCue%28Looping%29Local _player _drainedgameplaycuetag (Variables|InternalData|GetDrained%20Cue%20Parameters) false)))
        (:else
          (Utilities|String|LogString "Trying to remove Being Drained GC on Player that is not valid anymore?"))))))

(fn ConstructionScript)
